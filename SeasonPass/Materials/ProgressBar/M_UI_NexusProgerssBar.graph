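# Translucent UI material that draws the Season Pass Nexus progress bar with filled and empty regions.
# Start, end and animate scalars blend into a progress value that drives either a smoothstep edge or a rounded-box fill mask.
# MF_UI_Box SDFs supply rounded corners, inner glow and an optional pixel-sized outline that insets the fill.
# Color comes from an optional two-color gradient with HSV adjustment, a highlight, and separate filled and empty opacity.
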Material /SeasonPass/Materials/ProgressBar/M_UI_NexusProgerssBar
graph {
 properties {
  bAutomaticallySetUsageInEditor=false
  bUsedWithStaticMesh=false
  blendMode=BLEND_Translucent
  materialDomain=MD_UI
  usedShadingModels=Default Lit
 }
 nodes [
  n1 = VectorParameter() {desc=Highlight color, group=Color, parameterName=ColorB, sortPriority=2, defaultValue=<rgba 1 1 1 1>}
  n2 = NamedRerouteUsage()
  n3 = SmoothStep(Value=Multiply(A=OneMinus(None=@n2), B=NamedRerouteUsage())) {constMax=1.1, constMin=0.5}
  LinearInterpolate(
   A=Constant3Vector()
   B=LinearInterpolate(A=NamedRerouteUsage(), B=@n1.RGB, Alpha=Power(Base=@n3) {constExponent=3})
   Alpha=NamedRerouteUsage()
  )
  MaterialFunctionCall() {materialFunction=&DebugTimeSine}
  n4 = VectorParameter() {
   desc="Color for empty part of progress bar.\r\nAlpha defines opacity for empty part of progress bar."
   group=Color
   parameterName=EmptyColor
   sortPriority=10
   defaultValue=<rgba 0.75 0.75 0.75 0.2>
  }
  n5 = ScalarParameter() {defaultValue=1, desc="Alpha for filled part of progress bar.\r\n0 -1", group=Color, parameterName=FillAlpha, sliderMax=1}
  Multiply(A=LinearInterpolate(A=@n4.RGB, B=@n5, Alpha=NamedRerouteUsage()), B=NamedRerouteUsage())
  n6 = ScalarParameter() {
   defaultValue=1
   desc="Will use rounded box to show progress value. Not support snap to segments.\r\nbool"
   group=Progress_Layout
   parameterName=UseRoundedProgress
   sliderMax=1
  }
  n7 = MaterialFunctionCall(
   Size=Constant2Vector() {g=1, r=1}
   CornerRoundness=LinearInterpolate(
    A=ScalarParameter() {desc="Value for corner roundness. \r\n0 - no roundness\r\n1 - circle", group=Progress_Layout, parameterName=CornerRoundness}
    Alpha=@n6
   )
   OutlineThickness=NamedRerouteUsage()
   KeepAspectRatio=StaticBool() {value=true}
  ) {materialFunction=&MF_UI_Box}
  NamedRerouteDeclaration(None=SmoothStep(Value=@n7.SDF) {constMax=0, constMin=-1.25}) {name=Inner Glow, nodeColor=<rgba 0 0 0 1>}
  n8 = ScalarParameter() {
   desc="When progress value changes, it defines where change starts.\r\n0 - 1"
   group=Progress
   parameterName=ProgressStart
   sliderMax=1
   sortPriority=1
  }
  n9 = ScalarParameter() {
   defaultValue=1
   desc="Property for animating progress value change (from ProgressStart to ProgressEnd)\r\n0 - 1"
   group=Progress
   parameterName=AnimateProgress
   sliderMax=1
   sortPriority=3
  }
  n10 = ScalarParameter() {
   defaultValue=1
   desc="When progress value changes, it defines where change ends.\r\n0 - 1"
   group=Progress
   parameterName=ProgressEnd
   sliderMax=1
   sortPriority=2
  }
  n11 = LinearInterpolate(A=@n8, B=NamedRerouteDeclaration(None=@n9) {name=Progress End, nodeColor=<rgba 1 0.1172 0 1>}, Alpha=@n10)
  n12 = ComponentMask(None=Add(A=NamedRerouteUsage(), B=NamedRerouteDeclaration(None=@n11) {name=Progress, nodeColor=<rgba 0 0 0 1>}) {constB=0.5}) {r=true}
  NamedRerouteDeclaration(None=@n12) {name=ProgressInverted, nodeColor=<rgba 1 0.5859 0 1>}
  n13 = TextureCoordinate()
  n14 = Divide(A=Add(A=@n13) {constB=-0.5}, B=NamedRerouteUsage())
  NamedRerouteDeclaration(None=Multiply(A=Add(A=@n14) {constB=0.5}) {constB=-1}) {name=RemappedUVs, nodeColor=<rgba 0 0 0 1>}
  n15 = VectorParameter() {desc=Start color in gradient, group=Color, parameterName=Primary Color, sortPriority=1, defaultValue=<rgba 1 0.5034 0 1>}
  n16 = VectorParameter() {group=Color, parameterName=Secondary Color, sortPriority=2, defaultValue=<rgba 0.3515 0.0823 0.956 1>}
  n17 = SmoothStep(Min=NamedRerouteUsage(), Value=ComponentMask(None=NamedRerouteUsage()) {r=true}) {constMax=0}
  n18 = LinearInterpolate(
   A=@n15.RGB
   B=LinearInterpolate(A=@n16.RGB, B=@n15.RGB, Alpha=Power(Base=@n17))
   Alpha=ScalarParameter() {group=Color, parameterName=IsGradient, sortPriority=0}
  )
  n19 = RgbToHsv!(None=@n18)
  n20 = ComponentMask(None=@n19) {b=true}
  n21 = MaterialFunctionCall(
   R=ComponentMask(None=@n19) {r=true}
   G=Multiply(A=ComponentMask(None=@n19) {g=true}) {constB=0.9}
   B=Saturate(None=Add(A=@n20) {constB=0.65})
  ) {materialFunction=&AppendMany}
  NamedRerouteDeclaration(None=LinearInterpolate(A=@n18, B=HsvToRgb!(None=@n21.RGB), Alpha=ScalarParameter() {parameterName=ColorAdjustment})) {name=Base color, nodeColor=<rgba 0 0 0 1>}
  NamedRerouteDeclaration(None=@n13) {name=UVs, nodeColor=<rgba 0 0 0 1>}
  n22 = NamedRerouteUsage()
  n23 = OneMinus(None=NamedRerouteUsage())
  n24 = NamedRerouteUsage()
  n25 = AppendVector(A=Add(A=ComponentMask(None=@n22) {r=true}, B=Multiply(A=@n23, B=@n24)), B=ComponentMask(None=@n22) {g=true})
  n26 = NamedRerouteUsage()
  n27 = AppendVector(A=@n24, B=@n26)
  n28 = StaticBool() {value=true}
  NamedRerouteDeclaration(None=Multiply(
    A=MaterialFunctionCall(UVs=@n25, Size=@n27, CornerRoundness=@n26, KeepAspectRatio=@n28) {materialFunction=&MF_UI_Box}.Fill
    B=NamedRerouteUsage()
   )) {name=RoundedProgressMask, nodeColor=<rgba 0 0 0 1>}
  NamedRerouteDeclaration(None=@n7.Fill) {name=RoundedMask, nodeColor=<rgba 0 0 0 1>}
  n29 = ScalarParameter() {
   defaultValue=1
   desc="Will add outline for progress bar.\r\nDoesn't work correct with \"snap to segments\".\r\nbool"
   group=Outline
   parameterName=UseOutline
   sliderMax=1
   sortPriority=1
  }
  n30 = LinearInterpolate(
   B=ScalarParameter() {defaultValue=8, desc=Size of outline for rounded progress bar in px., group=Outline, parameterName=Outline_PX}
   Alpha=@n29
  )
  n31 = Constant() {r=1}
  n32 = Saturate(None=Multiply(A=MaterialFunctionCall!(Value PX=@n30, UV Pixel Size, X or Y=@n31) {materialFunction=/FortUILibrary/MaterialFunctions/Basics/MF_UI_PixelsToNormalized}.Normalized) {constB=1.2})
  NamedRerouteDeclaration(None=OneMinus(None=@n32)) {name=BarWithoutOutlineVertNormalized, nodeColor=<rgba 0 0 0 1>}
  n33 = MaterialFunctionCall(Value PX=@n30) {materialFunction=&MF_UI_PixelsToNormalized}
  NamedRerouteDeclaration(None=OneMinus(None=Saturate(None=@n33.Normalized))) {name=BarWithoutOutlineHorNormalized, nodeColor=<rgba 0 0 0 1>}
  NamedRerouteDeclaration(None=@n32) {name=OutlineVertNormalized, nodeColor=<rgba 0 0 0 1>}
  NamedRerouteDeclaration(None=OneMinus(None=@n7.Outline)) {name=OutlineMask, nodeColor=<rgba 0 0 0 1>}
  n34 = ScalarParameter() {defaultValue=0.1, group=Progress_Layout, parameterName=ProgressEdgeSharpness, sliderMax=1, sortPriority=2}
  n35 = ScalarParameter() {
   defaultValue=1
   desc="Will use rounded box to show progress value. Not support snap to segments.\r\nbool"
   group=Progress_Layout
   parameterName=UseRoundedProgress
   sliderMax=1
  }
  n36 = LinearInterpolate(A=Saturate(None=SmoothStep(Max=@n34, Value=@n12) {constMax=0}), B=NamedRerouteUsage(), Alpha=@n35)
  n37 = NamedRerouteUsage()
  n38 = ScalarParameter() {
   defaultValue=1
   desc="Will add outline for progress bar.\r\nDoesn't work correct with \"snap to segments\".\r\nbool"
   group=Outline
   parameterName=UseOutline
   sliderMax=1
   sortPriority=1
  }
  NamedRerouteDeclaration(None=LinearInterpolate(A=@n36, B=Multiply(A=@n36, B=@n37), Alpha=@n38)) {name=ProgressMask, nodeColor=<rgba 0 0 0 1>}
 ]
}
